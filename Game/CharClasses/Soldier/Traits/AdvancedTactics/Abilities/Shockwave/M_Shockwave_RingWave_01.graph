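# Unlit, two-sided translucent mesh-particle material for the ring wave of the Soldier's Shockwave ability.
# Three panned muzzle-texture samples are summed, with stepped time making the pattern jump rather than scroll.
# Particle dynamic parameters drive step rate, random U scale and erosion.
# The depth-faded result blends dark and bright colors into adjusted emissive and forms a mask with an optional bottom gradient.
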Material /Game/CharClasses/Soldier/Traits/AdvancedTactics/Abilities/Shockwave/M_Shockwave_RingWave_01
graph {
 properties {
  bEnableResponsiveAA=true
  bUsedWithMeshParticles=true
  blendMode=BLEND_Translucent
  pixelDepthOffsetMode=PDOM_Legacy
  shadingModel=MSM_Unlit
  twoSided=true
  usedShadingModels=Unlit
 }
 nodes [
  n1 = TextureCoordinate() {uTiling=1.5}
  n2 = AppendVector(A=DynamicParameter() {defaultValue=<rgba 17 1 1 0>, paramNames [Panner Speed, U Random, Param3, Erode]}.U Random, B=Constant() {r=1})
  n3 = Ceil(None=Multiply(A=Time(), B=DynamicParameter() {defaultValue=<rgba 17 1 1 0>, paramNames [Panner Speed, U Random, Param3, Erode]}.Panner Speed))
  n4 = TextureSample(UVs=Panner(Coordinate=Multiply(A=@n1, B=@n2), Time=@n3) {bFractionalPart=true, speedX=1.11}) {samplerType=SAMPLERTYPE_LinearColor, texture=/Game/Effects/Fort_Effects/Textures/Tile/T_Raygun_Muzzle_01}
  n5 = TextureCoordinate() {uTiling=-0.88}
  n6 = TextureSample(UVs=Panner(Coordinate=Multiply(A=@n5, B=@n2), Time=@n3) {bFractionalPart=true, speedX=-0.83}) {samplerType=SAMPLERTYPE_LinearColor, texture=&T_Raygun_Muzzle_01}
  n7 = Multiply(A=TextureCoordinate(), B=@n2)
  n8 = Multiply(A=TextureSample(UVs=Panner(Coordinate=@n7, Time=@n3) {bFractionalPart=true, speedX=4.48, speedY=0.332}) {samplerType=SAMPLERTYPE_LinearColor, texture=&T_Raygun_Muzzle_01}.B) {constB=10}
  n9 = ScalarParameter() {defaultValue=1, parameterName=UseBlueChannel?, sliderMax=1}
  n10 = Subtract(
   A=Add(A=Add(A=@n4.R, B=@n6.R), B=Multiply(A=@n8, B=@n9))
   B=DynamicParameter() {defaultValue=<rgba 17 1 1 0>, paramNames [Panner Speed, U Random, Param3, Erode]}.Erode
  )
  n11 = Clamp(None=Subtract(A=@n10))
  n12 = DepthFade(FadeDistance=ScalarParameter() {defaultValue=50, parameterName=Depth Fade Distance})
  MaterialFunctionCall(Emissive=LinearInterpolate(
    A=VectorParameter() {parameterName=Color Dark, defaultValue=<rgba 0.17 1.0848 2 1>}.RGB
    B=VectorParameter() {parameterName=Color Bright, defaultValue=<rgba 1.9459 2.4979 7 1>}.RGB
    Alpha=Multiply(A=@n11, B=@n12)
   ) {constAlpha=0}) {materialFunction=&MF_FX_EmissiveAdjustments}
  n13 = Multiply(A=@n10, B=@n12)
  n14 = MaterialFunctionCall() {materialFunction=&LinearGradient}
  n15 = ScalarParameter() {defaultValue=1, parameterName=BottomFalloff}
  n16 = Multiply(A=Power(Base=OneMinus(None=@n14.VGradient), Exp=@n15), B=@n15)
  StaticSwitchParameter(True=Multiply(A=@n13, B=Clamp(None=@n16)), False=@n13) {parameterName=UseBottomGradient}
 ]
}
